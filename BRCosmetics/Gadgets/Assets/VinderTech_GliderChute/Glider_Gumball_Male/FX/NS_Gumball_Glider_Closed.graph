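# Niagara burst effect for the Gumball glider, named for its closed state: a short-lived ball and eyes pop with a colored splat.
# Ball and Eyes meshes grow while curves drive their material parameters, with spawn groups sending the two eyes apart.
# Splat and Warp sprites expand and fade behind random camera offsets.
# User.VariantColor minus one indexes three-color arrays for Ball, Splat and Warp; the DEBUG emitter is disabled.
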NiagaraSystem /BRCosmetics/Gadgets/Assets/VinderTech_GliderChute/Glider_Gumball_Male/FX/NS_Gumball_Glider_Closed
dependencies {
 usedDataInterfaces [
  &NiagaraDataInterfaceArrayColor
  &NiagaraDataInterfaceColorCurve
  &NiagaraDataInterfaceCurve
  &NiagaraDataInterfaceVector2DCurve
 ]
 usedDynamicInputs [&Vector2DFromFloat]
 usedModules [
  &self:SystemSpawnScript.SystemScriptSource.SystemScriptGraph.NiagaraNodeAssignment_0.SetVariables_#1
  &SystemState
 ]
}
emitters {
 Ball {
  emitterUpdateScript [
   &EmitterState()
   &SpawnBurst_Instantaneous(Use Spawn Probability=false, UseLoopCountLimit=false)
  ]
  particleSpawnScript [&InitializeParticle(
    Color=&SelectLinearColorFromArray(
     Array Sampling Mode=::NewEnumerator1
     Direct Array Index=&MakeIntFromFloat(Float=&Add_Float(A=$User.VariantColor, B=-1))
     Direct Array Mode=::NewEnumerator1
     Color Selection Array=DataInterface() {propertyValues {ColorData [<rgba 1 0.203 0.4846 1>, <rgba 0.0462 0.0515 0.3455 1>, <rgba 0 0.3438 0.0039 1>]}}
     Random Seed=?
    )
    Lifetime=0.3
    Mesh Scale Mode=::NewEnumerator3
    Mesh Uniform Scale=110
    Position Offset=$System.Offset
    UsePositionOffset=true
    Write Material Random=true
   )]
  particleUpdateScript [
   &ParticleState()
   &SolveForcesAndVelocity(Clamp Velocity=false, Limit Acceleration=false)
   &DynamicMaterialParameters(
    Index 0 Param 2=5
    Index 0 Param 3=2
    Index 0 Param 4=&FloatFromCurve(FloatCurve=DataInterface() {propertyValues {Curve {keys [
         {arriveTangent=-3.7762, interpMode=RCIM_Cubic, leaveTangent=-3.7762, tangentMode=RCTM_User, time=0, value=1}
         {arriveTangent=-1.2694, interpMode=RCIM_Cubic, leaveTangent=-1.2694, tangentMode=RCTM_User, time=1, value=-1}
        ]}}})
    Index 1 Param 1=1
    Index 1 Param 2=1
    Module.Param0WriteEnabled=true
    Module.Param0WriteEnabled1=true
    Module.Param1WriteEnabled=true
    Module.Param1WriteEnabled1=true
    Module.Param2WriteEnabled=true
    Module.Param2WriteEnabled1=true
    Module.Param3WriteEnabled=true
    Module.Param3WriteEnabled1=true
    Write Parameter Index 1=true
   )
   &ScaleMeshSize(Scale Factor=&VectorFromFloat(Value=&FloatFromCurve(FloatCurve=DataInterface() {propertyValues {Curve {keys [
          {arriveTangent=1.0798, interpMode=RCIM_Cubic, leaveTangent=1.0798, tangentMode=RCTM_User, time=0, value=1}
          {interpMode=RCIM_Cubic, tangentMode=RCTM_User, time=1, value=2.5}
         ]}}})))
  ]
  rendererClasses [&NiagaraMeshRendererProperties]
  renderers [{rendererClass=&NiagaraMeshRendererProperties, rendererIndex=0}]
 }
 DEBUG {
  bEnabled=false
  emitterUpdateScript [
   &EmitterState()
   &SpawnBurst_Instantaneous(Use Spawn Probability=false, UseLoopCountLimit=false)
  ]
  particleSpawnScript [&InitializeParticle(Color Mode=::NewEnumerator0, Lifetime=3, Position Offset=$System.Offset, UsePositionOffset=true, Write Material Random=true)]
  particleUpdateScript [&ParticleState()]
  rendererClasses [&NiagaraSpriteRendererProperties]
  renderers [{rendererClass=&NiagaraSpriteRendererProperties}]
 }
 Eyes {
  emitterUpdateScript [
   &EmitterState()
   &SpawnBurst_Instantaneous(Spawn Count=2, Use Spawn Probability=false, UseLoopCountLimit=false)
  ]
  particleSpawnScript [
   &FN_SetSpawnGroup(Number of Spawn Groups=2)
   &InitializeParticle(
    Lifetime=&RandomRangeFloat(Minimum=0.2, Maximum=0.5)
    Mesh Scale Mode=::NewEnumerator3
    Mesh Uniform Scale=40
    Position Offset=&MaskVectorBySpawnGroup(Masked Vector=<-25 -30 75>, Passthrough Vector=<50 25 75>)
    UsePositionOffset=true
    Write Material Random=true
   )
   &InitialMeshOrientation(Use Rotation Vector=false)
   &AddVelocity(Velocity=&MaskVectorBySpawnGroup(Masked Vector=<-200 0 0>, Passthrough Vector=<200 0 0>))
  ]
  particleUpdateScript [
   &ParticleState()
   &SolveForcesAndVelocity(Clamp Velocity=false, Limit Acceleration=false)
   &DynamicMaterialParameters(
    Index 0 Param 1=&RandomRangeFloat(Minimum=0, Maximum=1)
    Index 0 Param 2=1
    Index 0 Param 3=2
    Index 0 Param 4=&FloatFromCurve(FloatCurve=DataInterface() {propertyValues {Curve {keys [
         {arriveTangent=-3.7762, interpMode=RCIM_Cubic, leaveTangent=-3.7762, tangentMode=RCTM_User, time=0, value=1}
         {arriveTangent=-1.2694, interpMode=RCIM_Cubic, leaveTangent=-1.2694, tangentMode=RCTM_User, time=1, value=-1}
        ]}}})
    Index 1 Param 1=0.35
    Index 1 Param 2=0.35
    Module.Param0WriteEnabled=true
    Module.Param0WriteEnabled1=true
    Module.Param1WriteEnabled=true
    Module.Param1WriteEnabled1=true
    Module.Param2WriteEnabled=true
    Module.Param2WriteEnabled1=true
    Module.Param3WriteEnabled=true
    Module.Param3WriteEnabled1=true
    Write Parameter Index 1=true
   )
   &ScaleMeshSize(Scale Factor=&VectorFromFloat(Value=&FloatFromCurve(FloatCurve=DataInterface() {propertyValues {Curve {keys [
          {arriveTangent=1.0798, interpMode=RCIM_Cubic, leaveTangent=1.0798, tangentMode=RCTM_User, time=0, value=1}
          {interpMode=RCIM_Cubic, tangentMode=RCTM_User, time=1, value=2.5}
         ]}}})))
   &CameraOffset(Camera Offset Amount=20)
  ]
  rendererClasses [&NiagaraMeshRendererProperties]
  renderers [{rendererClass=&NiagaraMeshRendererProperties, rendererIndex=0}]
 }
 Splat {
  emitterUpdateScript [
   &EmitterState()
   &SpawnRate(SpawnRate=5, Use Spawn Probability=true) {enabled=false}
   &SpawnBurst_Instantaneous(Use Spawn Probability=false, UseLoopCountLimit=false)
  ]
  particleSpawnScript [
   &InitializeParticle(
    Color=&SelectLinearColorFromArray(
     Array Sampling Mode=::NewEnumerator1
     Direct Array Index=&MakeIntFromFloat(Float=&Add_Float(A=$User.VariantColor, B=-1))
     Direct Array Mode=::NewEnumerator1
     Color Selection Array=DataInterface() {propertyValues {ColorData [<rgba 1 0.203 0.4846 1>, <rgba 0.0508 0.3703 1 1>, <rgba 0.0667 0.3993 0 1>]}}
     Random Seed=?
    )
    Lifetime=&RandomRangeFloat(Minimum=0.6, Maximum=0.8)
    Position Offset=$System.Offset
    Sprite Rotation Mode=::NewEnumerator2
    Sprite Size Mode=::NewEnumerator3
    Uniform Sprite Size=300
    UsePositionOffset=true
    Write Material Random=true
   )
   /SetVariables(Particles.Position=&Add_Vector(
     A=/Niagara/DynamicInputs/Helpers/SimulationPosition()
     B=&MakeVector(
      X=&FloatFromCurve(FloatCurve=DataInterface() {propertyValues {Curve {keys [<at 0.5 1>,<at 3 0>]}}}, CurveIndex=$System.Age, Scale Curve=7)
      Y=0
      Z=0
     )
    )) {enabled=false, isSetParameters=true}
  ]
  particleUpdateScript [
   &ParticleState()
   &SolveForcesAndVelocity(Clamp Velocity=false, Limit Acceleration=false)
   &ScaleColor(
    Scale Alpha=&FloatFromCurve(FloatCurve=DataInterface() {propertyValues {Curve {keys [
         {arriveTangent=-2.4316, interpMode=RCIM_Cubic, leaveTangent=-2.4316, tangentMode=RCTM_User, time=0.1, value=1}
         {arriveTangent=-0.9941, interpMode=RCIM_Cubic, leaveTangent=-0.9941, tangentMode=RCTM_User, time=0.4, value=0.3}
         {arriveTangent=-0.1071, interpMode=RCIM_Cubic, leaveTangent=-0.1071, tangentMode=RCTM_User, time=1, value=0.15}
        ]}}})
    ScaleA=true
    ScaleRGB=false
   )
   &ScaleSpriteSize(
    Uniform Curve Scale=1.2
    Uniform Curve Sprite Scale=DataInterface() {propertyValues {Curve {keys [
        {arriveTangent=1.2875, interpMode=RCIM_Cubic, leaveTangent=1.2875, tangentMode=RCTM_User, time=0, value=0.3}
        {interpMode=RCIM_Cubic, tangentMode=RCTM_User, time=1, value=1}
       ]}}}
   )
   &DynamicMaterialParameters(
    Module.Param0WriteEnabled1=true
    Module.Param1WriteEnabled1=true
    Module.Param2WriteEnabled1=true
    Module.Param3WriteEnabled1=true
    Write Parameter Index 0=false
    Write Parameter Index 1=true
   )
   &CameraOffset(Camera Offset Amount=&RandomRangeFloat(Minimum=0, Maximum=-50))
  ]
  rendererClasses [&NiagaraSpriteRendererProperties]
  renderers [{rendererClass=&NiagaraSpriteRendererProperties}]
 }
 Warp {
  emitterUpdateScript [
   &EmitterState()
   &SpawnRate(SpawnRate=5, Use Spawn Probability=true) {enabled=false}
   &SpawnBurst_Instantaneous(Spawn Count=2, Use Spawn Probability=false, UseLoopCountLimit=false)
  ]
  particleSpawnScript [&InitializeParticle(
    Color=&SelectLinearColorFromArray(
     Array Sampling Mode=::NewEnumerator1
     Direct Array Index=&MakeIntFromFloat(Float=&Add_Float(A=$User.VariantColor, B=-1))
     Direct Array Mode=::NewEnumerator1
     Color Selection Array=DataInterface() {propertyValues {ColorData [<rgba 1 0.203 0.4846 1>, <rgba 0.0747 0.0641 1 1>, <rgba 0.04 0.3021 0.0168 1>]}}
     Random Seed=?
    )
    Lifetime=&RandomRangeFloat(Minimum=0.6, Maximum=0.8)
    Position Offset=$System.Offset
    Sprite Rotation Mode=::NewEnumerator2
    Sprite Size Mode=::NewEnumerator3
    Uniform Sprite Size=300
    UsePositionOffset=true
    Write Material Random=true
   )]
  particleUpdateScript [
   &ParticleState()
   &SolveForcesAndVelocity(Clamp Velocity=false, Limit Acceleration=false)
   &ScaleColor(
    Scale Alpha=&FloatFromCurve(FloatCurve=DataInterface() {propertyValues {Curve {keys [
         {arriveTangent=-2.386, interpMode=RCIM_Cubic, leaveTangent=-2.386, tangentMode=RCTM_User, time=0.1, value=1}
         {arriveTangent=-0.9941, interpMode=RCIM_Cubic, leaveTangent=-0.9941, tangentMode=RCTM_User, time=0.4, value=0.3}
         {arriveTangent=-0.3618, interpMode=RCIM_Cubic, leaveTangent=-0.3618, tangentMode=RCTM_User, time=1, value=0.1}
        ]}}})
    ScaleA=true
    ScaleRGB=false
   )
   &ScaleSpriteSize(
    Uniform Curve Scale=1.2
    Uniform Curve Sprite Scale=DataInterface() {propertyValues {Curve {keys [
        {arriveTangent=1.2875, interpMode=RCIM_Cubic, leaveTangent=1.2875, tangentMode=RCTM_User, time=0, value=0.3}
        {arriveTangent=0.3825, interpMode=RCIM_Cubic, leaveTangent=0.3825, tangentMode=RCTM_User, time=1, value=1}
       ]}}}
   )
   &DynamicMaterialParameters(
    Index 1 Param 1=&FloatFromCurve(
     FloatCurve=DataInterface() {propertyValues {Curve {keys [{interpMode=RCIM_Cubic, tangentMode=RCTM_User, time=0, value=0}, {interpMode=RCIM_Cubic, tangentMode=RCTM_User, time=1, value=1}]}}}
     Scale Curve=5
    )
    Module.Param0WriteEnabled1=true
    Module.Param1WriteEnabled1=true
    Module.Param2WriteEnabled1=true
    Module.Param3WriteEnabled1=true
    Write Parameter Index 0=false
    Write Parameter Index 1=true
   )
   &CameraOffset(Camera Offset Amount=&RandomRangeFloat(Minimum=0, Maximum=-50))
  ]
  rendererClasses [&NiagaraSpriteRendererProperties]
  renderers [{rendererClass=&NiagaraSpriteRendererProperties}]
 }
}
summary {
 systemName=NS_Gumball_Glider_Closed
 emitters [
  {emitterName=Warp, rendererClasses [&NiagaraSpriteRendererProperties]}
  {emitterName=Splat, rendererClasses [&NiagaraSpriteRendererProperties]}
  {emitterName=Ball, rendererClasses [&NiagaraMeshRendererProperties]}
  {emitterName=Eyes, rendererClasses [&NiagaraMeshRendererProperties]}
  {bEnabled=false, emitterName=DEBUG, rendererClasses [&NiagaraSpriteRendererProperties]}
 ]
 userVariables [
  User.bIsFrontEnd {type=NiagaraBool}
  User.bIsFrontEndPreview {type=NiagaraBool}
  User.bIsFullyDeployed {type=NiagaraBool}
  User.ForwardDot {type=NiagaraFloat, defaultValue=0, description=""}
  User.SkeletalMesh {
   type=NiagaraDataInterfaceSkeletalMesh
   defaultValue {
    struct=NiagaraExt_VariableValue_DataInterface
    value {dataInterface=&self:NiagaraDataInterfaceSkeletalMesh_0}
   }
  }
  User.VariantColor {type=NiagaraFloat, defaultValue=0, description=""}
 ]
}
